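; Gameplay ability for GreaseClipDunmurry that hides and immobilizes the avatar player pawn while it is active.
; On activation it stops movement, sets the movement mode to None and hides the pawn in game under a dedicated gameplay tag.
; On end, cancelled or not, it sets movement to Falling and clears that tagged hide.
; For touch-input players, both events also toggle the kill feed HUD element's visibility.
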(event Ability|EventActivateAbility
  (bind _returnvalue (CallFunction|GetAvatarasFortPlayerPawn))
  (bind _charactermovement (Variables|/Script/Engine.Character|GetCharacterMovement _returnvalue))
  (Utilities|IsValid _returnvalue
    (:Is%20Valid
      (Components|Movement|StopMovementImmediately _charactermovement)
      (Pawn|Components|CharacterMovement|SetMovementMode _charactermovement "MOVE_None" 0)
      (Rendering|SetActorHiddenInGameWithTag _returnvalue true "(TagName=\"Pawn.HIdePlayerPawn.GreaseClipDunmurry\")" 0 true)
      (if (Inputs|IsUsingInputType "Touch" (Ability|GetPlayerControllerfromActorInfo))
        (HUD|SetHUDElementVisibility (BlueprintContext|GetFortHUDContext) (GameplayTags|MakeGameplayTagContainerfromTag "(TagName=\"HUD.Athena.KillFeed\")") true "None")))))

(event Ability|EventOnEndAbility
  (params (bWasCancelled :type Boolean))
  (bind _returnvalue (CallFunction|GetAvatarasFortPlayerPawn))
  (Utilities|IsValid _returnvalue
    (:Is%20Valid
      (Pawn|Components|CharacterMovement|SetMovementMode (Variables|/Script/Engine.Character|GetCharacterMovement _returnvalue) "MOVE_Falling" 0)
      (Rendering|SetActorHiddenInGameWithTag _returnvalue false "(TagName=\"Pawn.HIdePlayerPawn.GreaseClipDunmurry\")" 0 true)
      (if (Inputs|IsUsingInputType "Touch" (Ability|GetPlayerControllerfromActorInfo))
        (HUD|SetHUDElementVisibility (BlueprintContext|GetFortHUDContext) (GameplayTags|MakeGameplayTagContainerfromTag "(TagName=\"HUD.Athena.KillFeed\")") false "None")))))
